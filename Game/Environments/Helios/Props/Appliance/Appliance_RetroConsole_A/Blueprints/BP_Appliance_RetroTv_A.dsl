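; Retro TV prop Blueprint with two debug custom events for a spooky screen transition and screen turn-off.
; Each event writes the current game time to a scalar parameter on the static mesh's materials, BP_GetTime or BP_GetTime_TurnOff, as an effect timestamp.
; BeginPlay, overlap and Tick are empty, and the construction script only calls its parent.
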(event Custom|DebugSpookyScreenTurnOff
  (Rendering|Material|SetScalarParameterValueonMaterials (Variables|Debug|GetStaticMeshComponent) "BP_GetTime_TurnOff" (Utilities|Time|GetGameTimeinSeconds)))

(event Custom|DebugSpookyScreenTransition
  (Rendering|Material|SetScalarParameterValueonMaterials (Variables|Debug|GetStaticMeshComponent) "BP_GetTime" (Utilities|Time|GetGameTimeinSeconds)))

(event EventBeginPlay)

(event Collision|EventActorBeginOverlap
  (params (OtherActor :type Actor%20Object%20Reference)))

(event EventTick
  (params (DeltaSeconds :type Float%20%28single-precision%29)))

(fn ConstructionScript
  (|CallParentFunction))
