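# Screen material for a retro console TV prop, showing animated static that switches to a flipbook picture and can play a switch-off effect.
# It builds on the shared layered base material and world-effect functions, then adds scanline-modulated emissive compensated for eye adaptation.
# Blueprint-set start times drive the effects: the picture fades in from noise with pixelated UVs that sharpen, while switch-off brightens it and closes it from both vertical edges.
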Material /Game/Environments/Helios/Props/Appliance/Appliance_RetroConsole_A/Materials/M_ApplianNeRetroTV_Screen
graph {
 properties {
  bAutomaticallySetUsageInEditor=false
  bTangentSpaceNormal=false
  bUseMaterialAttributes=true
  bUsedWithInstancedStaticMeshes=true
  bUsedWithNanite=true
  bUsedWithSplineMeshes=true
  ditheredLODTransition=true
  pixelDepthOffsetMode=PDOM_Legacy
  shadingModel=MSM_FromMaterialExpression
  usedShadingModels=Default Lit | Subsurface | Clear Coat | Two Sided Foliage
 }
 nodes [
  n1 = MaterialFunctionCall() {materialFunction=/Game/Global/Base4Layers_2022/MaterialFunctions/MF_GigaMix_2023}
  MaterialExpressionReroute_70 = Reroute(None=@n1.Result)
  n2 = ScalarParameter() {
   desc=Documentation link
   group="0 (Expand)  Documentation (RMB - Copy display name) and Debug view"
   parameterName=https://confluence.it.epicgames.com/display/FORT/Fortnite+Base+Material+2022
  }
  n3 = SetMaterialAttributes!(MaterialAttributes=@MaterialExpressionReroute_70.MaterialAttributes, World Position Offset=Multiply(A=@n2)) {attributeSetTypes [#1]}
  MaterialExpressionReroute_71 = Reroute(None=@n3)
  MaterialExpressionReroute_69 = Reroute(None=@MaterialExpressionReroute_71.Output)
  n4 = MaterialFunctionCall!(Material Attributes=MaterialFunctionCall!(In=MaterialFunctionCall(Material_In=@MaterialExpressionReroute_69.MaterialAttributes) {materialFunction=&MF_FN_DestructibleDamage}.Material_Out) {materialFunction=/Game/Athena/WinterArt/Materials/MF_WPO_WindNew}.Result) {materialFunction=/Game/Creative/Sets/MilitaryBase/Props/MF_FanSpin}
  n5 = MaterialFunctionCall!(In=MaterialFunctionCall!(
    In=MaterialFunctionCall!(In=@n4.Result, UseSphericalMappingOverride) {materialFunction=/Game/Packages/Fortress_SharedMaterials/Base_Material_Functions/MF_CurieFallbackEffect}.Result
    Custom Mask
    Custom Rotation Axis
   ) {materialFunction=/Game/Athena/WinterArt/Materials/MF_WindWPO_v2}.Result) {materialFunction=/Game/Global/Base4Layers_2022/MaterialFunctions/MF_SnowAndIceArchitecture}
  MaterialExpressionReroute_68 = Reroute(None=@n5.Result)
  Composite!(None=@MaterialExpressionReroute_68.Result) {desc=Transform to World in case we will use RVT after this point, subgraphName=Tangent to World}
  Composite!(MaterialAttributes=@n1.Result) {subgraphName=Documentation Link}
  Composite(MaterialAttributes=@MaterialExpressionReroute_71.Output) {desc=Game Shared WPO and Fire Effect, subgraphName=WPO And Shared Functions}
  MaterialExpressionReroute_1 = Reroute(None=@n12.RVT H)
  MaterialExpressionReroute_3 = Reroute(None=@MaterialExpressionReroute_68.Result)
  n6 = GetMaterialAttributes(None=@MaterialExpressionReroute_3.Input) {attributeGetTypes [#2]}
  n7 = Transform(None=@n6.Normal)
  n8 = LinearInterpolate(A=Multiply(A=@n7) {constB=-1}, B=@n7, Alpha=Saturate(None=TwoSidedSign())) {desc="Fixes normals\r\non backfaces"}
  n9 = SetMaterialAttributes!(
   MaterialAttributes=@n6.MaterialAttributes
   Normal=StaticSwitchParameter(True=MaterialProxyReplace(Realtime=@n8, MaterialProxy=@n6.Normal), False=@n8) {parameterName=Force TangentSpaceNormalsFor HLOD ProxyReplace}
  ) {attributeSetTypes [#2]}
  MaterialExpressionReroute_2 = Reroute(None=@n9)
  n10 = MaterialFunctionCall!(In=MaterialFunctionCall!(In=@MaterialExpressionReroute_2.Output) {materialFunction=/Game/ArtTools/Tech_Art/DirectFX/Assets/MaterialFunctions/MF_DirectFX}.Result) {materialFunction=/Game/ArtTools/Tech_Art/Veneer/MaterialFunctions/MF_Veneer}
  MaterialExpressionReroute_8 = Reroute(None=@n10.Result)
  n11 = MaterialFunctionCall!(AO, RVT_Height=@MaterialExpressionReroute_1."RVT_Height (S)", For Buildings=@MaterialExpressionReroute_8."For Moss (MA)", For Moss) {materialFunction=/Game/Environments/Asteria/Rocks/Materials/MaterialFunctions/MF_JungleBiome}
  MaterialExpressionReroute_0 = Reroute(None=@n11.Buildings Jungl)
  n12 = MaterialFunctionCall!(
   In=@MaterialExpressionReroute_0.Result
   "RVT Height Map (Roughness)"=GetMaterialAttributes(None=@n10.Result) {attributeGetTypes [#3]}.Roughness
  ) {materialFunction=/Game/Global/Base4Layers_2022/MaterialFunctions/MF_RVTBlending}
  Composite!(InputPin=@n12.Result) {subgraphName=Mobile Opt}
  Composite!("For Moss (MA)"=@n10.Result, "RVT_Height (S)"=@n12.RVT H) {subgraphName="Seasonal "}
  Constant() {r=1}
  Constant() {r=1}
  n13 = ComponentMask(None=TextureCoordinate()) {g=true}
  n14 = Constant()
  n15 = Time()
  n16 = ScalarParameter() {defaultValue=10000000000, parameterName=BP_GetTime_TurnOff}
  n17 = Saturate(None=Multiply(A=Subtract(A=@n15, B=@n16)) {constB=1.6})
  n18 = Constant() {r=0.5}
  n19 = Constant() {r=1}
  n20 = Add(
   A=Saturate(None=SphereMask(A=@n13, B=@n14, Radius=@n17, Hardness=@n18))
   B=Saturate(None=SphereMask(A=@n13, B=@n19, Radius=@n17, Hardness=@n18))
  )
  Multiply(A=@n20) {constB=2}
  PinBase() {reroutePins [Input {expression=&self:MaterialExpressionReroute_3}]}
  PinBase!(Output=@n9) {reroutePins [Output {expression=&self:MaterialExpressionReroute_2}]}
  PinBase!() {reroutePins [
    "For Moss (MA)" {expression=&self:MaterialExpressionReroute_8}
    "RVT_Height (S)" {expression=&self:MaterialExpressionReroute_1}
   ]}
  PinBase!(Result=@n11.Buildings Jungl) {reroutePins [Result {expression=&self:MaterialExpressionReroute_0}]}
  PinBase!() {reroutePins [MaterialAttributes {expression=&self:MaterialExpressionReroute_70}]}
  PinBase!(Output=@n3) {reroutePins [Output {expression=&self:MaterialExpressionReroute_71}]}
  PinBase!() {reroutePins [MaterialAttributes {expression=&self:MaterialExpressionReroute_69}]}
  PinBase!(Result=@n5.Result) {reroutePins [Result {expression=&self:MaterialExpressionReroute_68}]}
  PinBase!() {reroutePins [InputPin {expression=&self:MaterialExpressionReroute_7}]}
  MaterialExpressionReroute_7 = Reroute(None=@n12.Result)
  n21 = GetMaterialAttributes(None=@MaterialExpressionReroute_7.InputPin)
  n22 = Constant() {r=0.5}
  n23 = Constant() {r=0.8}
  n24 = ShadingPathSwitch(
   Default=@MaterialExpressionReroute_7.InputPin
   Mobile=QualitySwitch(
    Default=@MaterialExpressionReroute_7.InputPin
    Low=SetMaterialAttributes!(MaterialAttributes=@n21.MaterialAttributes, Specular=@n22, Roughness=@n23) {attributeSetTypes [#4, #3]}
    High=@MaterialExpressionReroute_7.InputPin
    Medium=@MaterialExpressionReroute_7.InputPin
    Epic=@MaterialExpressionReroute_7.InputPin
   )
  )
  PinBase!(Output=@n24) {reroutePins [Output {expression=&self:MaterialExpressionReroute_6}]}
  n25 = Subtract(A=Time(), B=ScalarParameter() {defaultValue=10000000000, parameterName=BP_GetTime})
  Reroute(None=@n25)
  MaterialExpressionReroute_6 = Reroute(None=@n24)
  n26 = GetMaterialAttributes(None=MaterialFunctionCall(MA In=@MaterialExpressionReroute_6.Output) {materialFunction=&MF_CreativeMoveTool}.MA Out) {attributeGetTypes [#5]}
  n27 = Add(B=MaterialFunctionCall!(ScanLine_Tiling_1, Scanline_PannerSpeed_1, ScanLine_Tiling, Scanline_PannerSpeed, Scanline_Rotation, UVs) {materialFunction=/Game/Global/Materials/MaterialGym/Materials/HolographicTexture/MF_HolographicTexture_ScanLines}.Result) {constA=0.5}
  n28 = Multiply(A=Time()) {constB=0.45}
  n29 = Constant() {r=5}
  n30 = TextureSample(UVs=MaterialFunctionCall("Animation  Phase (0-1)"=Frac(None=@n28), Number of Rows=@n29, Number of Columns=@n29) {materialFunction=&FlipBook}.UVs) {
   samplerSource=SSM_Wrap_WorldGroupSettings
   texture=/Game/Environments/Helios/Props/Appliance/Appliance_RetroConsole_A/Textures/T_Appliance_RetroTVScreen_A_Flipbook_Noise
  }
  n31 = NamedRerouteDeclaration(None=@n25) {name=BP_GetTime, nodeColor=<rgba 1 0 0.0469 1>}
  n32 = TextureCoordinate()
  n33 = Constant() {r=15}
  n34 = Divide(A=Floor(None=Multiply(A=@n32, B=@n33)), B=@n33)
  n35 = NamedRerouteUsage()
  n36 = Floor(None=Saturate(None=Multiply(A=@n35) {constB=2.2}))
  n37 = TextureSample(UVs=MaterialFunctionCall(
    "Animation  Phase (0-1)"=Multiply(B=@n31) {constA=0.42}
    Number of Rows=@n29
    Number of Columns=@n29
    UVs=LinearInterpolate(A=@n34, B=@n32, Alpha=@n36)
   ) {materialFunction=&FlipBook}.UVs) {
   samplerSource=SSM_Wrap_WorldGroupSettings
   texture=/Game/Environments/Helios/Props/Appliance/Appliance_RetroConsole_A/Textures/T_Appliance_RetroTVScreen_A_Flipbook
  }
  n38 = Multiply(A=NamedRerouteUsage())
  n39 = Multiply(A=@n37.RGB, B=LinearInterpolate(Alpha=Saturate(None=@n38)) {constA=2})
  n40 = Saturate(None=Multiply(A=@n31) {constB=10})
  n41 = Ceil(None=@n17)
  n42 = Multiply(A=Multiply(A=Saturate(None=@n27), B=LinearInterpolate(A=@n30.RGB, B=@n39, Alpha=@n40)), B=Add(B=Multiply(A=@n41) {constB=3.5}) {constA=1})
  n43 = Saturate(None=@n20)
  n44 = Power(Base=LinearInterpolate(B=@n39, Alpha=@n40)) {constExponent=3}
  SetMaterialAttributes!(
   MaterialAttributes=@n26.MaterialAttributes
   Emissive Color=Divide(A=LinearInterpolate(A=@n42, Alpha=@n43) {constB=0}, B=EyeAdaptation().EyeAdaptation)
   Base Color=Add(A=@n26.Base Color, B=Multiply(A=@n44) {constB=2})
  ) {attributeSetTypes [#6, #5]}
 ]
}
